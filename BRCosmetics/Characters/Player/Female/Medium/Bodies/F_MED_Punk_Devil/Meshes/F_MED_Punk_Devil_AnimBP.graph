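# Post-process style cosmetic AnimBlueprint for the F_MED_Punk_Devil body, adding secondary motion and pelvis/thigh correctives to a pose copied from the attached parent mesh.
# Chained spring-driven Anim Dynamics nodes move boot zippers, pulls, sleeves and back shirt hem.
# Additive Modify Bone nodes then offset pelvis and left thigh-twist bones, weighted by ThighRotationLeftAlpha/RightAlpha, which are not declared here.
# Left, right and front shirt-hem dynamics form a separate chain unconnected to Output Pose.
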AnimBlueprint /BRCosmetics/Characters/Player/Female/Medium/Bodies/F_MED_Punk_Devil/Meshes/F_MED_Punk_Devil_AnimBP
target_skeleton=&/F_MED_Punk_Devil_Skeleton
graphs [
 AnimGraph {nodes [
   n1 = Anim Dynamics - Bone: shirtBottom_l(Alpha=1, ExternalForce="0,0,0") {
    angularDampingOverride=1
    angularSpringConstant=1500
    bAngularSpring=true
    bGravityOverrideInSimSpace=true
    bOverrideAngularDamping=true
    gravityScale=0.35
    simulationSpace=RootRelative
    alphaBoolBlend {blendOption=Linear}
    boundBone {boneName=shirtBottom_l}
    componentLinearAccScale {z=0.6}
    componentLinearVelScale {z=0.6}
    physicsBodyDefinitions [{
      sphereCollisionRadius=0
      boundBone {boneName=shirtBottom_l}
      boxExtents=<10 20 20>
      constraintSetup {angularLimitsMax=<0 15 0>, angularLimitsMin=<0 -15 -40>, angularTarget=<0 0 0>}
      localJointOffset=<10 0 0>
     }]
   }
   n2 = Anim Dynamics - Bone: shirtBottom_r(ComponentPose=@n1, Alpha=1, ExternalForce="0,0,0") {
    angularDampingOverride=1
    angularSpringConstant=1500
    bAngularSpring=true
    bGravityOverrideInSimSpace=true
    bOverrideAngularDamping=true
    gravityScale=0.35
    simulationSpace=RootRelative
    alphaBoolBlend {blendOption=Linear}
    boundBone {boneName=shirtBottom_r}
    componentLinearAccScale {z=0.6}
    componentLinearVelScale {z=0.6}
    physicsBodyDefinitions [{
      sphereCollisionRadius=0
      boundBone {boneName=shirtBottom_r}
      boxExtents=<10 20 20>
      constraintSetup {angularLimitsMax=<0 15 0>, angularLimitsMin=<0 -15 -40>, angularTarget=<0 0 0>}
      localJointOffset=<10 0 0>
     }]
   }
   Anim Dynamics - Bone: shirtBottom_fnt(ComponentPose=@n2, Alpha=1, ExternalForce="0,0,0") {
    angularDampingOverride=1
    angularSpringConstant=1500
    bAngularSpring=true
    bGravityOverrideInSimSpace=true
    bOverrideAngularDamping=true
    gravityScale=0.35
    simulationSpace=RootRelative
    alphaBoolBlend {blendOption=Linear}
    boundBone {boneName=shirtBottom_fnt}
    componentLinearAccScale {z=0.6}
    componentLinearVelScale {z=0.6}
    physicsBodyDefinitions [{
      sphereCollisionRadius=0
      boundBone {boneName=shirtBottom_fnt}
      boxExtents=<10 20 20>
      constraintSetup {angularLimitsMax=<0 15 0>, angularLimitsMin=<0 -15 -40>, angularTarget=<0 0 0>}
      localJointOffset=<10 0 0>
     }]
   }
   n3 = Anim Dynamics - Bone: bootZipper_r(ComponentPose=Local To Component(LocalPose=Copy Pose From Mesh() {bUseAttachedParent=true}), Alpha=1, ExternalForce="0,0,0") {
    angularDampingOverride=1
    angularSpringConstant=1500
    bAngularSpring=true
    bGravityOverrideInSimSpace=true
    bOverrideAngularDamping=true
    gravityScale=0.35
    simulationSpace=RootRelative
    alphaBoolBlend {blendOption=Linear}
    boundBone {boneName=bootZipper_r}
    physicsBodyDefinitions [{
      sphereCollisionRadius=0
      boundBone {boneName=bootZipper_r}
      boxExtents=<10 20 20>
      constraintSetup {angularLimitsMax=<0 15 0>, angularLimitsMin=<0 -15 -40>, angularTarget=<0 0 0>}
      localJointOffset=<10 0 0>
     }]
   }
   n4 = Anim Dynamics - Bone: bootPull_r(ComponentPose=@n3, Alpha=1, ExternalForce="0,0,0") {
    angularDampingOverride=1
    angularSpringConstant=1500
    bAngularSpring=true
    bGravityOverrideInSimSpace=true
    bOverrideAngularDamping=true
    gravityScale=0.35
    simulationSpace=RootRelative
    alphaBoolBlend {blendOption=Linear}
    boundBone {boneName=bootPull_r}
    physicsBodyDefinitions [{
      sphereCollisionRadius=0
      boundBone {boneName=bootPull_r}
      boxExtents=<10 20 20>
      constraintSetup {angularLimitsMax=<0 0 0>, angularLimitsMin=<0 0 -40>, angularTarget=<0 0 0>}
      localJointOffset=<10 0 0>
     }]
   }
   n5 = Anim Dynamics - Bone: bootPull_l(ComponentPose=@n4, Alpha=1, ExternalForce="0,0,0") {
    angularDampingOverride=1
    angularSpringConstant=1500
    bAngularSpring=true
    bGravityOverrideInSimSpace=true
    bOverrideAngularDamping=true
    gravityScale=0.35
    simulationSpace=RootRelative
    alphaBoolBlend {blendOption=Linear}
    boundBone {boneName=bootPull_l}
    physicsBodyDefinitions [{
      sphereCollisionRadius=0
      boundBone {boneName=bootPull_l}
      boxExtents=<10 20 20>
      constraintSetup {angularLimitsMax=<0 0 0>, angularLimitsMin=<0 0 -40>, angularTarget=<0 0 0>}
      localJointOffset=<10 0 0>
     }]
   }
   n6 = Anim Dynamics - Bone: bootZipper_l(ComponentPose=@n5, Alpha=1, ExternalForce="0,0,0") {
    angularDampingOverride=1
    angularSpringConstant=1500
    bAngularSpring=true
    bGravityOverrideInSimSpace=true
    bOverrideAngularDamping=true
    gravityScale=0.35
    simulationSpace=RootRelative
    alphaBoolBlend {blendOption=Linear}
    boundBone {boneName=bootZipper_l}
    physicsBodyDefinitions [{
      sphereCollisionRadius=0
      boundBone {boneName=bootZipper_l}
      boxExtents=<10 20 20>
      constraintSetup {angularLimitsMax=<0 15 0>, angularLimitsMin=<0 -15 -40>, angularTarget=<0 0 0>}
      localJointOffset=<10 0 0>
     }]
   }
   n7 = Anim Dynamics - Bone: sleeve_bt_r(ComponentPose=@n6, Alpha=1, ExternalForce="0,0,0") {
    angularDampingOverride=1
    angularSpringConstant=5000
    bAngularSpring=true
    bGravityOverrideInSimSpace=true
    bOverrideAngularDamping=true
    gravityScale=0.35
    simulationSpace=RootRelative
    alphaBoolBlend {blendOption=Linear}
    alphaScaleBiasClamp {scale=0.2}
    boundBone {boneName=sleeve_bt_r}
    componentLinearAccScale {z=0.6}
    componentLinearVelScale {z=0.6}
    physicsBodyDefinitions [{
      sphereCollisionRadius=0
      boundBone {boneName=sleeve_bt_r}
      boxExtents=<10 20 20>
      constraintSetup {angularLimitsMax=<45 20 0>, angularLimitsMin=<-45 -20 -75>, angularTarget=<1 0 0>}
      localJointOffset=<7 0 0>
     }]
   }
   n8 = Anim Dynamics - Bone: sleeve_bb_r(ComponentPose=@n7, Alpha=1, ExternalForce="0,0,0") {
    angularDampingOverride=1
    angularSpringConstant=5000
    bAngularSpring=true
    bGravityOverrideInSimSpace=true
    bOverrideAngularDamping=true
    gravityScale=0.35
    simulationSpace=RootRelative
    alphaBoolBlend {blendOption=Linear}
    alphaScaleBiasClamp {scale=0.2}
    boundBone {boneName=sleeve_bb_r}
    componentLinearAccScale {z=0.6}
    componentLinearVelScale {z=0.6}
    physicsBodyDefinitions [{
      sphereCollisionRadius=0
      boundBone {boneName=sleeve_bb_r}
      boxExtents=<10 20 20>
      constraintSetup {angularLimitsMax=<45 20 0>, angularLimitsMin=<-45 -20 -75>, angularTarget=<1 0 0>}
      localJointOffset=<7 0 0>
     }]
   }
   n9 = Anim Dynamics - Bone: sleeve_fb_r(ComponentPose=@n8, Alpha=1, ExternalForce="0,0,0") {
    angularDampingOverride=1
    angularSpringConstant=5000
    bAngularSpring=true
    bGravityOverrideInSimSpace=true
    bOverrideAngularDamping=true
    gravityScale=0.35
    simulationSpace=RootRelative
    alphaBoolBlend {blendOption=Linear}
    alphaScaleBiasClamp {scale=0.2}
    boundBone {boneName=sleeve_fb_r}
    componentLinearAccScale {z=0.6}
    componentLinearVelScale {z=0.6}
    physicsBodyDefinitions [{
      sphereCollisionRadius=0
      boundBone {boneName=sleeve_fb_r}
      boxExtents=<10 20 20>
      constraintSetup {angularLimitsMax=<45 20 0>, angularLimitsMin=<-45 -20 -75>, angularTarget=<1 0 0>}
      localJointOffset=<7 0 0>
     }]
   }
   n10 = Anim Dynamics - Bone: sleeve_ft_r(ComponentPose=@n9, Alpha=1, ExternalForce="0,0,0") {
    angularDampingOverride=1
    angularSpringConstant=5000
    bAngularSpring=true
    bGravityOverrideInSimSpace=true
    bOverrideAngularDamping=true
    gravityScale=0.35
    simulationSpace=RootRelative
    alphaBoolBlend {blendOption=Linear}
    alphaScaleBiasClamp {scale=0.2}
    boundBone {boneName=sleeve_ft_r}
    componentLinearAccScale {z=0.6}
    componentLinearVelScale {z=0.6}
    physicsBodyDefinitions [{
      sphereCollisionRadius=0
      boundBone {boneName=sleeve_ft_r}
      boxExtents=<10 20 20>
      constraintSetup {angularLimitsMax=<45 20 0>, angularLimitsMin=<-45 -20 -75>, angularTarget=<1 0 0>}
      localJointOffset=<7 0 0>
     }]
   }
   n11 = Anim Dynamics - Bone: sleeve_bt_l(ComponentPose=@n10, Alpha=1, ExternalForce="0,0,0") {
    angularDampingOverride=1
    angularSpringConstant=5000
    bAngularSpring=true
    bGravityOverrideInSimSpace=true
    bOverrideAngularDamping=true
    gravityScale=0.35
    simulationSpace=RootRelative
    alphaBoolBlend {blendOption=Linear}
    alphaScaleBiasClamp {scale=0.2}
    boundBone {boneName=sleeve_bt_l}
    componentLinearAccScale {z=0.6}
    componentLinearVelScale {z=0.6}
    physicsBodyDefinitions [{
      sphereCollisionRadius=0
      boundBone {boneName=sleeve_bt_l}
      boxExtents=<10 20 20>
      constraintSetup {angularLimitsMax=<45 20 0>, angularLimitsMin=<-45 -20 -75>, angularTarget=<1 0 0>}
      localJointOffset=<7 0 0>
     }]
   }
   n12 = Anim Dynamics - Bone: sleeve_bb_l(ComponentPose=@n11, Alpha=1, ExternalForce="0,0,0") {
    angularDampingOverride=1
    angularSpringConstant=5000
    bAngularSpring=true
    bGravityOverrideInSimSpace=true
    bOverrideAngularDamping=true
    gravityScale=0.35
    simulationSpace=RootRelative
    alphaBoolBlend {blendOption=Linear}
    alphaScaleBiasClamp {scale=0.2}
    boundBone {boneName=sleeve_bb_l}
    componentLinearAccScale {z=0.6}
    componentLinearVelScale {z=0.6}
    physicsBodyDefinitions [{
      sphereCollisionRadius=0
      boundBone {boneName=sleeve_bb_l}
      boxExtents=<10 20 20>
      constraintSetup {angularLimitsMax=<45 20 0>, angularLimitsMin=<-45 -20 -75>, angularTarget=<1 0 0>}
      localJointOffset=<7 0 0>
     }]
   }
   n13 = Anim Dynamics - Bone: sleeve_fb_l(ComponentPose=@n12, Alpha=1, ExternalForce="0,0,0") {
    angularDampingOverride=1
    angularSpringConstant=5000
    bAngularSpring=true
    bGravityOverrideInSimSpace=true
    bOverrideAngularDamping=true
    gravityScale=0.35
    simulationSpace=RootRelative
    alphaBoolBlend {blendOption=Linear}
    alphaScaleBiasClamp {scale=0.2}
    boundBone {boneName=sleeve_fb_l}
    componentLinearAccScale {z=0.6}
    componentLinearVelScale {z=0.6}
    physicsBodyDefinitions [{
      sphereCollisionRadius=0
      boundBone {boneName=sleeve_fb_l}
      boxExtents=<10 20 20>
      constraintSetup {angularLimitsMax=<45 20 0>, angularLimitsMin=<-45 -20 -75>, angularTarget=<1 0 0>}
      localJointOffset=<7 0 0>
     }]
   }
   n14 = Anim Dynamics - Bone: sleeve_ft_l(ComponentPose=@n13, Alpha=1, ExternalForce="0,0,0") {
    angularDampingOverride=1
    angularSpringConstant=5000
    bAngularSpring=true
    bGravityOverrideInSimSpace=true
    bOverrideAngularDamping=true
    gravityScale=0.35
    simulationSpace=RootRelative
    alphaBoolBlend {blendOption=Linear}
    alphaScaleBiasClamp {scale=0.2}
    boundBone {boneName=sleeve_ft_l}
    componentLinearAccScale {z=0.6}
    componentLinearVelScale {z=0.6}
    physicsBodyDefinitions [{
      sphereCollisionRadius=0
      boundBone {boneName=sleeve_ft_l}
      boxExtents=<10 20 20>
      constraintSetup {angularLimitsMax=<45 20 0>, angularLimitsMin=<-45 -20 -75>, angularTarget=<1 0 0>}
      localJointOffset=<7 0 0>
     }]
   }
   n15 = Anim Dynamics - Bone: shirtBottom_bck(ComponentPose=@n14, Alpha=1, ExternalForce="0,0,0") {
    angularDampingOverride=1
    angularSpringConstant=1500
    bAngularSpring=true
    bGravityOverrideInSimSpace=true
    bOverrideAngularDamping=true
    gravityScale=0.35
    simulationSpace=RootRelative
    alphaBoolBlend {blendOption=Linear}
    boundBone {boneName=shirtBottom_bck}
    componentLinearAccScale {z=0.6}
    componentLinearVelScale {z=0.6}
    physicsBodyDefinitions [{
      sphereCollisionRadius=0
      boundBone {boneName=shirtBottom_bck}
      boxExtents=<10 20 20>
      constraintSetup {angularLimitsMax=<0 15 0>, angularLimitsMin=<0 -15 -40>, angularTarget=<0 0 0>}
      localJointOffset=<10 0 0>
     }]
   }
   n16 = Get ThighRotationLeftAlpha() -> (ThighRotationLeftAlpha:float)
   n17 = "Transform (Modify) Bone - Bone: pelvis_l"(ComponentPose=@n15, Alpha=@n16, Translation="-10,0,0", Rotation="0,0,0", Scale="1,1.5,1") {
    scaleMode=BMM_Additive
    translationMode=BMM_Additive
    translationSpace=BCS_ParentBoneSpace
    alphaBoolBlend {blendOption=Linear}
    alphaScaleBiasClamp {scale=0.8}
    boneToModify {boneName=pelvis_l}
    scale {y=1.5}
    translation {x=-10}
   }
   n18 = "Transform (Modify) Bone - Bone: pelvis_mid_l"(ComponentPose=@n17, Alpha=@n16, Translation="-5,0,0", Rotation="0,0,0", Scale="1,1,1") {
    scaleMode=BMM_Additive
    translationMode=BMM_Additive
    translationSpace=BCS_ParentBoneSpace
    alphaBoolBlend {blendOption=Linear}
    alphaScaleBiasClamp {scale=0.8}
    boneToModify {boneName=pelvis_mid_l}
    translation {x=-5}
   }
   n19 = "Transform (Modify) Bone - Bone: pelvis_side_l"(ComponentPose=@n18, Alpha=@n16, Translation="0,0,-8", Rotation="0,0,0", Scale="1,1,1") {
    scaleMode=BMM_Additive
    translationMode=BMM_Additive
    translationSpace=BCS_ParentBoneSpace
    alphaBoolBlend {blendOption=Linear}
    alphaScaleBiasClamp {scale=0.8}
    boneToModify {boneName=pelvis_side_l}
    translation {z=-8}
   }
   n20 = "Transform (Modify) Bone - Bone: thigh_twist_01_bot_l"(ComponentPose=@n19, Alpha=@n16, Translation="0,-6,0", Rotation="0,0,0", Scale="1,1,1") {
    scaleMode=BMM_Additive
    translationMode=BMM_Additive
    translationSpace=BCS_ParentBoneSpace
    alphaBoolBlend {blendOption=Linear}
    boneToModify {boneName=thigh_twist_01_bot_l}
    translation {y=-6}
   }
   n21 = Get ThighRotationRightAlpha() -> (ThighRotationRightAlpha:float)
   n22 = "Transform (Modify) Bone - Bone: pelvis_r"(ComponentPose=@n20, Alpha=@n21, Translation="-12,0,0", Rotation="0,0,0", Scale="1,1.5,1") {
    scaleMode=BMM_Additive
    translationMode=BMM_Additive
    translationSpace=BCS_ParentBoneSpace
    alphaBoolBlend {blendOption=Linear}
    boneToModify {boneName=pelvis_r}
    scale {y=1.5}
    translation {x=-12}
   }
   n23 = "Transform (Modify) Bone - Bone: pelvis_mid_r"(ComponentPose=@n22, Alpha=@n21, Translation="-5,0,0", Rotation="0,0,0", Scale="1,1,1") {
    scaleMode=BMM_Additive
    translationMode=BMM_Additive
    translationSpace=BCS_ParentBoneSpace
    alphaBoolBlend {blendOption=Linear}
    boneToModify {boneName=pelvis_mid_r}
    translation {x=-5}
   }
   n24 = "Transform (Modify) Bone - Bone: pelvis_side_r"(ComponentPose=@n23, Alpha=@n21, Translation="0,0,8", Rotation="0,0,0", Scale="1,1,1") {
    scaleMode=BMM_Additive
    translationMode=BMM_Additive
    translationSpace=BCS_ParentBoneSpace
    alphaBoolBlend {blendOption=Linear}
    boneToModify {boneName=pelvis_side_r}
    translation {z=8}
   }
   Output Pose(Result=Component To Local(ComponentPose=@n24)) {layerGroup=None}
  ]}
 EventGraph {links [], nodes []}
]
variables []
