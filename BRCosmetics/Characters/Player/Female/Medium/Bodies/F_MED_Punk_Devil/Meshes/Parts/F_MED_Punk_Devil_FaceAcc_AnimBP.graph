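# Animation Blueprint for the Punk Devil female face accessory, simulating left, center and right hair bone chains.
# The parent pose, layered with body-mesh clavicles, enters a PreSim Control Rig driven by movement, weapon, aiming and backpack states.
# Two cached, sequential AnimDynamics passes follow, the first using spherical limits on clavicles and spine, before a Post Control Rig outputs.
# The EventGraph updates state flags and computes gravity from a parameter asset.
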AnimBlueprint /BRCosmetics/Characters/Player/Female/Medium/Bodies/F_MED_Punk_Devil/Meshes/Parts/F_MED_Punk_Devil_FaceAcc_AnimBP
target_skeleton=&/F_MED_Punk_Devil_FaceAcc_Skeleton
graphs [
 AnimGraph {nodes [
   n1 = Control Rig(Source=Use cached pose 'CR_Cache_Blend'()) {controlRigAssetReference {blueprintRigClass=&/M_MED_Punk_Devil_FaceAcc_Post_CtrlRig.M_MED_Punk_Devil_FaceAcc_Post_CtrlRig_C}}
   Output Pose(Result=@n1) {layerGroup=None}
   n2 = Anim Dynamics - Bone: dyn_L_hair_1!(
    ComponentPose=Local To Component(LocalPose=Use cached pose 'CR_Cache_Input'())
    bAlphaBoolEnabled:bool=True
    GravityOverride:Vector="0,0,-980"
    ExternalForce="0,0,0"
   ) -> (Pose) {
    alphaInputType=Bool
    angularDampingOverride=1
    angularSpringConstant=14000
    bAngularSpring=true
    bOverrideAngularDamping=true
    bUseGravityOverride=true
    gravityScale=0.35
    simulationSpace=RootRelative
    boundBone {boneName=dyn_L_hair_1}
    componentAppliedLinearAccClamp=<2000 2000 2000>
    componentLinearAccScale=<0.5 0.5 0.5>
    componentLinearVelScale=<0.5 0.5 0.5>
    gravityOverride {z=-980}
    physicsBodyDefinitions [{
      boundBone {boneName=dyn_L_hair_1}
      boxExtents=<20 30 30>
      constraintSetup {angularLimitsMax=<0 20 1>, angularLimitsMin=<0 1 -1>, angularTarget=<1 0 0>}
      localJointOffset=<4 0 0>
     }]
    simSpaceSettings {maxAngularAcceleration=60, maxAngularVelocity=7, simSpaceAngularAlpha=0.4}
   }
   n3 = Anim Dynamics - Bone: dyn_L_hair_2!(ComponentPose=@n2, bAlphaBoolEnabled:bool=True, GravityOverride:Vector="0,0,-980", ExternalForce="0,0,0") -> (Pose) {
    alphaInputType=Bool
    angularDampingOverride=1
    angularSpringConstant=14000
    bAngularSpring=true
    bOverrideAngularDamping=true
    bUseGravityOverride=true
    bUseSphericalLimits=true
    gravityScale=0.35
    simulationSpace=RootRelative
    boundBone {boneName=dyn_L_hair_2}
    componentAppliedLinearAccClamp=<2000 2000 2000>
    componentLinearAccScale=<0.5 0.5 0.5>
    componentLinearVelScale=<0.5 0.5 0.5>
    gravityOverride {z=-980}
    physicsBodyDefinitions [{
      collisionType=CustomSphere
      sphereCollisionRadius=4.5
      boundBone {boneName=dyn_L_hair_2}
      boxExtents=<20 30 30>
      constraintSetup {angularLimitsMax=<0 20 2>, angularLimitsMin=<0 1 -2>, angularTarget=<1 0 0>}
      localJointOffset=<4 0 0>
     }]
    simSpaceSettings {maxAngularAcceleration=60, maxAngularVelocity=7, simSpaceAngularAlpha=0.4}
    sphericalLimits [{limitRadius=19.5, drivingBone {boneName=clavicle_l}, sphereLocalOffset=<3.4624 10.0404 -10.4866>}]
   }
   n4 = Anim Dynamics - Bone: dyn_L_hair_3!(ComponentPose=@n3, bAlphaBoolEnabled:bool=True, GravityOverride:Vector="0,0,-980", ExternalForce="0,0,0") -> (Pose) {
    alphaInputType=Bool
    angularDampingOverride=1
    angularSpringConstant=14000
    bAngularSpring=true
    bOverrideAngularDamping=true
    bUseGravityOverride=true
    bUseSphericalLimits=true
    gravityScale=0.35
    simulationSpace=RootRelative
    boundBone {boneName=dyn_L_hair_3}
    componentAppliedLinearAccClamp=<2000 2000 2000>
    componentLinearAccScale=<0.5 0.5 0.5>
    componentLinearVelScale=<0.5 0.5 0.5>
    gravityOverride {z=-980}
    physicsBodyDefinitions [{
      collisionType=CustomSphere
      sphereCollisionRadius=1.5
      boundBone {boneName=dyn_L_hair_3}
      boxExtents=<20 30 30>
      constraintSetup {angularLimitsMax=<0 20 2>, angularLimitsMin=<0 1 -2>, angularTarget=<1 0 0>}
      localJointOffset=<4 0 0>
     }]
    simSpaceSettings {maxAngularAcceleration=60, maxAngularVelocity=7, simSpaceAngularAlpha=0.4}
    sphericalLimits [{limitRadius=19.5, drivingBone {boneName=clavicle_l}, sphereLocalOffset=<7.0243 1.811 -19.6215>}]
   }
   n5 = Anim Dynamics - Bone: dyn_C_hair_1!(ComponentPose=@n4, bAlphaBoolEnabled:bool=True, GravityOverride:Vector="0,0,-980", ExternalForce="0,0,0") -> (Pose) {
    alphaInputType=Bool
    angularDampingOverride=1
    angularSpringConstant=14000
    bAngularSpring=true
    bOverrideAngularDamping=true
    bUseGravityOverride=true
    gravityScale=0.35
    simulationSpace=RootRelative
    boundBone {boneName=dyn_C_hair_1}
    componentAppliedLinearAccClamp=<2000 2000 2000>
    componentLinearAccScale=<0.5 0.5 0.5>
    componentLinearVelScale=<0.5 0.5 0.5>
    gravityOverride {z=-980}
    physicsBodyDefinitions [{
      boundBone {boneName=dyn_C_hair_1}
      boxExtents=<20 30 30>
      constraintSetup {angularLimitsMax=<0 20 1>, angularLimitsMin=<0 1 -1>, angularTarget=<1 0 0>}
      localJointOffset=<4 0 0>
     }]
    simSpaceSettings {maxAngularAcceleration=60, maxAngularVelocity=7, simSpaceAngularAlpha=0.4}
   }
   n6 = Anim Dynamics - Bone: dyn_C_hair_2!(ComponentPose=@n5, bAlphaBoolEnabled:bool=True, GravityOverride:Vector="0,0,-980", ExternalForce="0,0,0") -> (Pose) {
    alphaInputType=Bool
    angularDampingOverride=1
    angularSpringConstant=14000
    bAngularSpring=true
    bOverrideAngularDamping=true
    bUseGravityOverride=true
    bUseSphericalLimits=true
    gravityScale=0.35
    simulationSpace=RootRelative
    boundBone {boneName=dyn_C_hair_2}
    componentAppliedLinearAccClamp=<2000 2000 2000>
    componentLinearAccScale=<0.5 0.5 0.5>
    componentLinearVelScale=<0.5 0.5 0.5>
    gravityOverride {z=-980}
    physicsBodyDefinitions [{
      collisionType=CustomSphere
      sphereCollisionRadius=4
      boundBone {boneName=dyn_C_hair_2}
      boxExtents=<20 30 30>
      constraintSetup {angularLimitsMax=<0 20 2>, angularLimitsMin=<0 1 -2>, angularTarget=<1 0 0>}
      localJointOffset=<4 0 0>
     }]
    simSpaceSettings {maxAngularAcceleration=60, maxAngularVelocity=7, simSpaceAngularAlpha=0.4}
    sphericalLimits [{limitRadius=20, drivingBone {boneName=spine_05}, sphereLocalOffset=<-7.2391 6.5936 0>}]
   }
   n7 = Anim Dynamics - Bone: dyn_C_hair_3!(ComponentPose=@n6, bAlphaBoolEnabled:bool=True, GravityOverride:Vector="0,0,-980", ExternalForce="0,0,0") -> (Pose) {
    alphaInputType=Bool
    angularDampingOverride=1
    angularSpringConstant=14000
    bAngularSpring=true
    bOverrideAngularDamping=true
    bUseGravityOverride=true
    bUseSphericalLimits=true
    gravityScale=0.35
    simulationSpace=RootRelative
    boundBone {boneName=dyn_C_hair_3}
    componentAppliedLinearAccClamp=<2000 2000 2000>
    componentLinearAccScale=<0.5 0.5 0.5>
    componentLinearVelScale=<0.5 0.5 0.5>
    gravityOverride {z=-980}
    physicsBodyDefinitions [{
      collisionType=CustomSphere
      sphereCollisionRadius=2
      boundBone {boneName=dyn_C_hair_3}
      boxExtents=<20 30 30>
      constraintSetup {angularLimitsMax=<0 20 2>, angularLimitsMin=<0 1 -2>, angularTarget=<1 0 0>}
      localJointOffset=<4 0 0>
     }]
    simSpaceSettings {maxAngularAcceleration=60, maxAngularVelocity=7, simSpaceAngularAlpha=0.4}
    sphericalLimits [{limitRadius=20, drivingBone {boneName=spine_05}, sphereLocalOffset=<-12.0218 4.2036 0>}]
   }
   n8 = Anim Dynamics - Bone: dyn_R_hair_1!(ComponentPose=@n7, bAlphaBoolEnabled:bool=True, GravityOverride:Vector="0,0,-980", ExternalForce="0,0,0") -> (Pose) {
    alphaInputType=Bool
    angularDampingOverride=1
    angularSpringConstant=14000
    bAngularSpring=true
    bOverrideAngularDamping=true
    bUseGravityOverride=true
    gravityScale=0.35
    simulationSpace=RootRelative
    boundBone {boneName=dyn_R_hair_1}
    componentAppliedLinearAccClamp=<2000 2000 2000>
    componentLinearAccScale=<0.5 0.5 0.5>
    componentLinearVelScale=<0.5 0.5 0.5>
    gravityOverride {z=-980}
    physicsBodyDefinitions [{
      boundBone {boneName=dyn_R_hair_1}
      boxExtents=<20 30 30>
      constraintSetup {angularLimitsMax=<0 20 1>, angularLimitsMin=<0 1 -1>, angularTarget=<1 0 0>}
      localJointOffset=<4 0 0>
     }]
    simSpaceSettings {maxAngularAcceleration=60, maxAngularVelocity=7, simSpaceAngularAlpha=0.4}
   }
   n9 = Anim Dynamics - Bone: dyn_R_hair_2!(ComponentPose=@n8, bAlphaBoolEnabled:bool=True, GravityOverride:Vector="0,0,-980", ExternalForce="0,0,0") -> (Pose) {
    alphaInputType=Bool
    angularDampingOverride=1
    angularSpringConstant=14000
    bAngularSpring=true
    bOverrideAngularDamping=true
    bUseGravityOverride=true
    bUseSphericalLimits=true
    gravityScale=0.35
    simulationSpace=RootRelative
    boundBone {boneName=dyn_R_hair_2}
    componentAppliedLinearAccClamp=<2000 2000 2000>
    componentLinearAccScale=<0.5 0.5 0.5>
    componentLinearVelScale=<0.5 0.5 0.5>
    gravityOverride {z=-980}
    physicsBodyDefinitions [{
      collisionType=CustomSphere
      sphereCollisionRadius=5
      boundBone {boneName=dyn_R_hair_2}
      boxExtents=<20 30 30>
      constraintSetup {angularLimitsMax=<0 20 2>, angularLimitsMin=<0 1 -2>, angularTarget=<1 0 0>}
      localJointOffset=<4 0 0>
     }]
    simSpaceSettings {maxAngularAcceleration=60, maxAngularVelocity=7, simSpaceAngularAlpha=0.4}
    sphericalLimits [{limitRadius=20, drivingBone {boneName=clavicle_r}, sphereLocalOffset=<-2.7145 -5.3898 13.5094>}]
   }
   n10 = Anim Dynamics - Bone: dyn_R_hair_3!(ComponentPose=@n9, bAlphaBoolEnabled:bool=True, GravityOverride:Vector="0,0,-980", ExternalForce="0,0,0") -> (Pose) {
    alphaInputType=Bool
    angularDampingOverride=1
    angularSpringConstant=14000
    bAngularSpring=true
    bOverrideAngularDamping=true
    bUseGravityOverride=true
    bUseSphericalLimits=true
    gravityScale=0.35
    simulationSpace=RootRelative
    boundBone {boneName=dyn_R_hair_3}
    componentAppliedLinearAccClamp=<2000 2000 2000>
    componentLinearAccScale=<0.5 0.5 0.5>
    componentLinearVelScale=<0.5 0.5 0.5>
    gravityOverride {z=-980}
    physicsBodyDefinitions [{
      collisionType=CustomSphere
      sphereCollisionRadius=2
      boundBone {boneName=dyn_R_hair_3}
      boxExtents=<20 30 30>
      constraintSetup {angularLimitsMax=<0 20 2>, angularLimitsMin=<0 1 -2>, angularTarget=<1 0 0>}
      localJointOffset=<4 0 0>
     }]
    simSpaceSettings {maxAngularAcceleration=60, maxAngularVelocity=7, simSpaceAngularAlpha=0.4}
    sphericalLimits [{limitRadius=20, drivingBone {boneName=clavicle_r}, sphereLocalOffset=<-5.9094 -5.0786 17.8455>}]
   }
   Save cached pose 'CR_Cache_Dynamics'(Pose=Component To Local(ComponentPose=@n10)) {cacheName=CR_Cache_Dynamics}
   n11 = Get Body_Mesh()
   n12 = Control Rig!(
    Source=Layered blend per bone(BasePose=Copy Pose From Mesh() {bCopyCurves=true, bUseAttachedParent=true}, BlendPoses_0=Copy Pose From Mesh(SourceMeshComponent=@n11)) {layerSetup [{branchFilters [{boneName=clavicle_l}, {boneName=clavicle_r}]}]}
    Alpha
    SkydivingFloating:double=Get isDivingFloating float() -> (isDivingFloating float:double)
    Crouching:double=Get IsCrouching() -> (IsCrouching:double)
    PA Swing:bool=Get bIsPlayingMeleeAnim()
    PistolEquipeed:double=Get IsPistolEquipped() -> (IsPistolEquipped:double)
    Is Targeting:bool=Get bIsTargeting()
    Jumping:double=Get IsJumping() -> (IsJumping:double)
    WearongPack:bool=Get isWearingPack() -> (isWearingPack:bool)
    ParachuteOpen:double=Get IsParachuteOpened() -> (IsParachuteOpened:double)
   ) -> (Pose) {controlRigAssetReference {blueprintRigClass=&/M_MED_Punk_Devil_FaceAcc_PreSim_CtrlRig.M_MED_Punk_Devil_FaceAcc_PreSim_CtrlRig_C}}
   Save cached pose 'CR_Cache_Input'(Pose=@n12) {cacheName=CR_Cache_Input}
   n13 = Anim Dynamics - Bone: dyn_L_hair_1(
    ComponentPose=Local To Component(LocalPose=Use cached pose 'CR_Cache_Dynamics'())
    bAlphaBoolEnabled=True
    GravityOverride="0,0,-980"
    ExternalForce="0,0,0"
   ) {
    alphaInputType=Bool
    angularDampingOverride=1
    angularSpringConstant=10000
    bAngularSpring=true
    bOverrideAngularDamping=true
    bUseGravityOverride=true
    gravityScale=0.35
    simulationSpace=RootRelative
    boundBone {boneName=dyn_L_hair_1}
    componentAppliedLinearAccClamp=<2000 2000 2000>
    componentLinearAccScale=<0.5 0.5 0.5>
    componentLinearVelScale=<0.5 0.5 0.5>
    gravityOverride {z=-980}
    physicsBodyDefinitions [{
      boundBone {boneName=dyn_L_hair_1}
      boxExtents=<20 30 30>
      constraintSetup {angularLimitsMax=<0 0 2>, angularLimitsMin=<0 0 -2>, angularTarget=<1 0 0>}
      localJointOffset=<2 0 0>
     }]
    simSpaceSettings {simSpaceAngularAlpha=0.4}
   }
   n14 = Anim Dynamics - Bone: dyn_L_hair_2(ComponentPose=@n13, bAlphaBoolEnabled=True, GravityOverride="0,0,-980", ExternalForce="0,0,0") {
    alphaInputType=Bool
    angularDampingOverride=1
    angularSpringConstant=10000
    bAngularSpring=true
    bOverrideAngularDamping=true
    bUseGravityOverride=true
    gravityScale=0.35
    simulationSpace=RootRelative
    boundBone {boneName=dyn_L_hair_2}
    componentAppliedLinearAccClamp=<2000 2000 2000>
    componentLinearAccScale=<0.5 0.5 0.5>
    componentLinearVelScale=<0.5 0.5 0.5>
    gravityOverride {z=-980}
    physicsBodyDefinitions [{
      boundBone {boneName=dyn_L_hair_2}
      boxExtents=<20 30 30>
      constraintSetup {angularLimitsMax=<0 0 2>, angularLimitsMin=<0 0 -2>, angularTarget=<1 0 0>}
      localJointOffset=<2 0 0>
     }]
    simSpaceSettings {simSpaceAngularAlpha=0.4}
   }
   n15 = Anim Dynamics - Bone: dyn_L_hair_3(ComponentPose=@n14, bAlphaBoolEnabled=True, GravityOverride="0,0,-980", ExternalForce="0,0,0") {
    alphaInputType=Bool
    angularDampingOverride=1
    angularSpringConstant=10000
    bAngularSpring=true
    bOverrideAngularDamping=true
    bUseGravityOverride=true
    gravityScale=0.35
    simulationSpace=RootRelative
    boundBone {boneName=dyn_L_hair_3}
    componentAppliedLinearAccClamp=<2000 2000 2000>
    componentLinearAccScale=<0.5 0.5 0.5>
    componentLinearVelScale=<0.5 0.5 0.5>
    gravityOverride {z=-980}
    physicsBodyDefinitions [{
      boundBone {boneName=dyn_L_hair_3}
      boxExtents=<20 30 30>
      constraintSetup {angularLimitsMax=<0 0 2>, angularLimitsMin=<0 0 -2>, angularTarget=<1 0 0>}
      localJointOffset=<2 0 0>
     }]
    simSpaceSettings {simSpaceAngularAlpha=0.4}
   }
   n16 = Anim Dynamics - Bone: dyn_C_hair_1(ComponentPose=@n15, bAlphaBoolEnabled=True, GravityOverride="0,0,-980", ExternalForce="0,0,0") {
    alphaInputType=Bool
    angularDampingOverride=1
    angularSpringConstant=10000
    bAngularSpring=true
    bOverrideAngularDamping=true
    bUseGravityOverride=true
    gravityScale=0.35
    simulationSpace=RootRelative
    boundBone {boneName=dyn_C_hair_1}
    componentAppliedLinearAccClamp=<2000 2000 2000>
    componentLinearAccScale=<0.5 0.5 0.5>
    componentLinearVelScale=<0.5 0.5 0.5>
    gravityOverride {z=-980}
    physicsBodyDefinitions [{
      boundBone {boneName=dyn_C_hair_1}
      boxExtents=<20 30 30>
      constraintSetup {angularLimitsMax=<0 0 2>, angularLimitsMin=<0 0 -2>, angularTarget=<1 0 0>}
      localJointOffset=<2 0 0>
     }]
    simSpaceSettings {simSpaceAngularAlpha=0.4}
   }
   n17 = Anim Dynamics - Bone: dyn_C_hair_2(ComponentPose=@n16, bAlphaBoolEnabled=True, GravityOverride="0,0,-980", ExternalForce="0,0,0") {
    alphaInputType=Bool
    angularDampingOverride=1
    angularSpringConstant=10000
    bAngularSpring=true
    bOverrideAngularDamping=true
    bUseGravityOverride=true
    gravityScale=0.35
    simulationSpace=RootRelative
    boundBone {boneName=dyn_C_hair_2}
    componentAppliedLinearAccClamp=<2000 2000 2000>
    componentLinearAccScale=<0.5 0.5 0.5>
    componentLinearVelScale=<0.5 0.5 0.5>
    gravityOverride {z=-980}
    physicsBodyDefinitions [{
      boundBone {boneName=dyn_C_hair_2}
      boxExtents=<20 30 30>
      constraintSetup {angularLimitsMax=<0 0 2>, angularLimitsMin=<0 0 -2>, angularTarget=<1 0 0>}
      localJointOffset=<2 0 0>
     }]
    simSpaceSettings {simSpaceAngularAlpha=0.4}
   }
   n18 = Anim Dynamics - Bone: dyn_C_hair_3(ComponentPose=@n17, bAlphaBoolEnabled=True, GravityOverride="0,0,-980", ExternalForce="0,0,0") {
    alphaInputType=Bool
    angularDampingOverride=1
    angularSpringConstant=10000
    bAngularSpring=true
    bOverrideAngularDamping=true
    bUseGravityOverride=true
    gravityScale=0.35
    simulationSpace=RootRelative
    boundBone {boneName=dyn_C_hair_3}
    componentAppliedLinearAccClamp=<2000 2000 2000>
    componentLinearAccScale=<0.5 0.5 0.5>
    componentLinearVelScale=<0.5 0.5 0.5>
    gravityOverride {z=-980}
    physicsBodyDefinitions [{
      boundBone {boneName=dyn_C_hair_3}
      boxExtents=<20 30 30>
      constraintSetup {angularLimitsMax=<0 0 2>, angularLimitsMin=<0 0 -2>, angularTarget=<1 0 0>}
      localJointOffset=<2 0 0>
     }]
    simSpaceSettings {simSpaceAngularAlpha=0.4}
   }
   n19 = Anim Dynamics - Bone: dyn_R_hair_1(ComponentPose=@n18, bAlphaBoolEnabled=True, GravityOverride="0,0,-980", ExternalForce="0,0,0") {
    alphaInputType=Bool
    angularDampingOverride=1
    angularSpringConstant=10000
    bAngularSpring=true
    bOverrideAngularDamping=true
    bUseGravityOverride=true
    gravityScale=0.35
    simulationSpace=RootRelative
    boundBone {boneName=dyn_R_hair_1}
    componentAppliedLinearAccClamp=<2000 2000 2000>
    componentLinearAccScale=<0.5 0.5 0.5>
    componentLinearVelScale=<0.5 0.5 0.5>
    gravityOverride {z=-980}
    physicsBodyDefinitions [{
      boundBone {boneName=dyn_R_hair_1}
      boxExtents=<20 30 30>
      constraintSetup {angularLimitsMax=<0 0 2>, angularLimitsMin=<0 0 -2>, angularTarget=<1 0 0>}
      localJointOffset=<2 0 0>
     }]
    simSpaceSettings {simSpaceAngularAlpha=0.4}
   }
   n20 = Anim Dynamics - Bone: dyn_R_hair_2(ComponentPose=@n19, bAlphaBoolEnabled=True, GravityOverride="0,0,-980", ExternalForce="0,0,0") {
    alphaInputType=Bool
    angularDampingOverride=1
    angularSpringConstant=10000
    bAngularSpring=true
    bOverrideAngularDamping=true
    bUseGravityOverride=true
    gravityScale=0.35
    simulationSpace=RootRelative
    boundBone {boneName=dyn_R_hair_2}
    componentAppliedLinearAccClamp=<2000 2000 2000>
    componentLinearAccScale=<0.5 0.5 0.5>
    componentLinearVelScale=<0.5 0.5 0.5>
    gravityOverride {z=-980}
    physicsBodyDefinitions [{
      boundBone {boneName=dyn_R_hair_2}
      boxExtents=<20 30 30>
      constraintSetup {angularLimitsMax=<0 0 2>, angularLimitsMin=<0 0 -2>, angularTarget=<1 0 0>}
      localJointOffset=<2 0 0>
     }]
    simSpaceSettings {simSpaceAngularAlpha=0.4}
   }
   n21 = Anim Dynamics - Bone: dyn_R_hair_3(ComponentPose=@n20, bAlphaBoolEnabled=True, GravityOverride="0,0,-980", ExternalForce="0,0,0") {
    alphaInputType=Bool
    angularDampingOverride=1
    angularSpringConstant=10000
    bAngularSpring=true
    bOverrideAngularDamping=true
    bUseGravityOverride=true
    gravityScale=0.35
    simulationSpace=RootRelative
    boundBone {boneName=dyn_R_hair_3}
    componentAppliedLinearAccClamp=<2000 2000 2000>
    componentLinearAccScale=<0.5 0.5 0.5>
    componentLinearVelScale=<0.5 0.5 0.5>
    gravityOverride {z=-980}
    physicsBodyDefinitions [{
      boundBone {boneName=dyn_R_hair_3}
      boxExtents=<20 30 30>
      constraintSetup {angularLimitsMax=<0 0 2>, angularLimitsMin=<0 0 -2>, angularTarget=<1 0 0>}
      localJointOffset=<2 0 0>
     }]
    simSpaceSettings {simSpaceAngularAlpha=0.4}
   }
   Save cached pose 'CR_Cache_Blend'(Pose=Component To Local(ComponentPose=@n21)) {cacheName=CR_Cache_Blend}
  ]}
 EventGraph {nodes [
   n1 = Event BlueprintUpdateAnimation()
   > Set Delta Time X(Delta Time X:double=@n1.DeltaTimeX) -> (Output_Get:double)
   > n2 = Sequence!(execute) -> (then_0, then_1, then_2:Exec, then_3:Exec, then_4:Exec, then_5:Exec, then_6:Exec, then_7:Exec, then_8:Exec)
   > n3 = "Gravity Override State Function (Parameter Asset) (Nativized)"(
    CharacterPartAnimInstance=Self-Reference()
    GameTimeInSeconds=GetGameTimeInSeconds()
    Parameters=&/M_MED_Punk_Devil_FaceAcc_Parameters
   )
   > Set GravityOverride(GravityOverride=@n3)
   > Set RBAN Settings Out(RBAN Settings Out=@n3.OutRigidBodyAnimNodeParameters)
   n4 = Branch(execute=@n2.then_2, Condition=OR Boolean(A=Get isWearingPack(), B=Get bIsParachuteOpened()))
   > Set toogleAnimDynamics(toogleAnimDynamics:bool=false) -> (Output_Get:bool)
   > Set EnabledynamicsPack(EnabledynamicsPack:bool=true) -> (Output_Get:bool)
   Set toogleAnimDynamics(execute=@n4.else, toogleAnimDynamics=true)
   > Set EnabledynamicsPack(EnabledynamicsPack=false)
   Set disableUpperLimbsCorrection(
    execute=@n2.then_4
    disableUpperLimbsCorrection:bool=OR Boolean!(A=Get bIsSkydiveDiveMode(), B=Get bIsSkydiveFloating(), C:bool=Get bIsParachuteOpened()) -> (ReturnValue)
   ) -> (Output_Get:bool)
   > n5 = Branch(Condition=OR Boolean!(A=Get bIsSkydiveDiveMode(), B=Get bIsSkydiveFloating()) -> (ReturnValue))
   > Set IsSkyDive(IsSkyDive:double=1) -> (Output_Get:double)
   Set IsSkyDive(execute=@n5.else, IsSkyDive=0)
   n6 = OR Boolean!(A=Get bIsSkydiveFloating(), B=Get bIsSkydiveDiveMode(), C:bool=Get bIsParachuteOpened(), D:bool=Get bIsSkyDiving()) -> (ReturnValue)
   n7 = Get bIsJumping()
   n8 = Get bIsFalling()
   n9 = Branch(execute=@n2.then_8, Condition=AND Boolean(A=OR Boolean!(A=@n7, B=@n8) -> (ReturnValue), B=NOT Boolean(A=@n6)))
   > Set IsJumping(IsJumping:double=1) -> (Output_Get:double)
   Set IsJumping(execute=@n9.else, IsJumping=0)
   Event OnParachuteOpened()
   > Set IsParachuteOpened(IsParachuteOpened:double=1) -> (Output_Get:double)
   Event OnEndPickaxeSwing() {is=title K2Node_Event}
   > Set IsSwing(IsSwing:double=0) -> (Output_Get:double)
   Event OnBeginCrouching()
   > Set IsCrouching(IsCrouching:double=1) -> (Output_Get:double)
   Event OnEndCrouching()
   > Set IsCrouching(IsCrouching=0)
   n10 = Event OnCharacterCustomizationLoaded()
   > Set Body_Mesh(Body_Mesh=@n10.CustomBaseMesh)
   n11 = CheckForBackbling()
   > Set isWearingPack(isWearingPack:bool=OR Boolean!(A=@n11, B=@n11.bIsBackblingACape, C:bool=@n11.bIsBackblingLong) -> (ReturnValue)) -> (Output_Get:bool)
   > Set isWearingCape(isWearingCape=@n11.bIsBackblingLong)
   Event OnBeginSkydiving()
   > Set isDivingFloating float(isDivingFloating float:double=1) -> (Output_Get:double)
   Event OnEndSkydiving()
   > Set isDivingFloating float(isDivingFloating float=0)
   Event OnWeaponChanged()
   > Set isAssaultRifleEquipped float(isAssaultRifleEquipped float:double=0) -> (Output_Get:double)
   > Set IsOutlowShotgunEquipped(IsOutlowShotgunEquipped:double=0) -> (Output_Get:double)
   > Set IsSMGEquipped(IsSMGEquipped:double=0) -> (Output_Get:double)
   > Set IsPistolEquipped(IsPistolEquipped:double=0) -> (Output_Get:double)
   > n12 = Switch on EFortWeaponCoreAnimation!(execute, Selection:EFortWeaponCoreAnimation Enum=Get WeaponCoreAnim!() -> (WeaponCoreAnim)) -> (
    Melee:Exec
    Pistol:Exec
    Shotgun:Exec
    PaperBlueprint:Exec
    Rifle:Exec
    MeleeOneHand:Exec
    MachinePistol:Exec
    RocketLauncher:Exec
    GrenadeLauncher:Exec
    GoingCommando:Exec
    AssaultRifle:Exec
    TacticalShotgun:Exec
    SniperRifle:Exec
    TrapPlacement:Exec
    ShoulderLauncher:Exec
    AbilityDecoTool:Exec
    Crossbow:Exec
    C4:Exec
    RemoteControl:Exec
    DualWield:Exec
    AR_BullPup:Exec
    AR_ForwardGrip:Exec
    MedPackPaddles:Exec
    SMG_P90:Exec
    AR_DrumGun:Exec
    Consumable_Small:Exec
    Consumable_Large:Exec
    Balloon:Exec
    MountedTurret:Exec
    CreativeTool:Exec
    ExplosiveBow:Exec
    AshtonIndigo:Exec
    AshtonChicago:Exec
    MeleeDualWield:Exec
    Unarmed:Exec
   )
   Set isAssaultRifleEquipped float(execute=@n12.AssaultRifle, isAssaultRifleEquipped float=1)
   Set IsPistolEquipped(execute=@n12.Pistol, IsPistolEquipped=1)
   Set IsSMGEquipped(execute=@n12.SMG_P90, IsSMGEquipped=1)
   Set IsOutlowShotgunEquipped(execute=@n12.TacticalShotgun, IsOutlowShotgunEquipped=1)
   Event OnParachuteClosed()
   > Set IsParachuteOpened(IsParachuteOpened=0)
   Event OnBeginPickaxeSwing() {is=title K2Node_Event}
   > Set IsSwing(IsSwing=1)
   Event BlueprintInitializeAnimation()
   > Set Skeletal_Mesh(Skeletal_Mesh:Skeletal Mesh Component Object Reference=GetOwnerSkeletalMeshComponent()) -> (Output_Get:Skeletal Mesh Component Object Reference)
   Get bIsTacticalSprinting()
  ]}
]
variables [
 Body_Mesh {type=SkeletalMeshComponent}
 Delta Time X {type=number}
 disableUpperLimbsCorrection {type=boolean}
 EnabledynamicsPack {type=boolean}
 GravityOverride {type=Vector}
 isAssaultRifleEquipped float {type=number}
 IsCrouching {type=number}
 isDivingFloating float {type=number}
 IsJumping {type=number}
 IsOutlowShotgunEquipped {type=number}
 IsParachuteOpened {type=number}
 IsPistolEquipped {type=number}
 IsSkyDive {type=number}
 IsSMGEquipped {type=number}
 IsSwing {type=number}
 isWearingCape {type=boolean}
 isWearingPack {type=boolean}
 RBAN Settings Out {type=FortRigidBodyAnimNodeParameters}
 Skeletal_Mesh {type=SkeletalMeshComponent}
 toogleAnimDynamics {type=boolean}
]
