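# Control Rig for the female Punk Devil FaceAcc part that keeps long hair bones from passing through the chest.
# Chest collider bones under spine_05 act as planes: each hair bone's position is converted to collider space, local Y is clamped to -100..0, then converted back.
# Clamps chain through two or three colliders before setting translation.
# Afterward, hair_2_l aims at corrected hair_3_side2_l, and hair_2_side_r blends partway toward its clamped position.
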ControlRigBlueprint /BRCosmetics/Characters/Player/Female/Medium/Bodies/F_MED_Punk_Devil/Meshes/Parts/F_MED_Punk_Devil_FaceAcc_CtrlRig
graphs [
 RigVMFunctionLibrary {links [], nodes []}
 RigVMModel {nodes [
   n1 = &RigUnit_BeginExecution()
   n2 = &RigUnit_GetBoneTransform(Bone=chestCol_c)
   n3 = &RigUnit_GetBoneTransform(Bone=chestCol_r)
   n4 = &RigUnit_GetBoneTransform(Bone=chestCol_l)
   n5 = &RigVMFunction_MathTransformMakeRelative(Global=&RigUnit_GetBoneTransform(Bone=hairTipMid_side_l).Transform, Parent=@n4.Transform)
   n6 = &RigVMFunction_MathTransformMakeRelative(
    Global=(Rotation=(X=0,Y=0,Z=0,W=1),Translation=(X=0,Y=0,Z=0),Scale3D=(X=1,Y=1,Z=1))
    Parent=@n3.Transform
    Global.Translation=&RigVMFunction_MathTransformMul(
     A=(Rotation=(X=0,Y=0,Z=0,W=1),Translation=(X=0,Y=0,Z=0),Scale3D=(X=1,Y=1,Z=1))
     B=@n4.Transform
     A.Translation.Y=/Script/RigVM.RigVMFunction_MathFloatClamp(Maximum=0, Minimum=-100, Value=@n5.Local.Translation.Y) {is=struct RigVMUnitNode, template="Clamp::Execute(in Value,in Minimum,in Maximum,out Result)"}.Result
     A.Translation.X=@n5.Local.Translation.X
     A.Translation.Z=@n5.Local.Translation.Z
    ) {template="Multiply::Execute(in A,in B,out Result)"}.Result.Translation
   )
   n7 = &RigVMFunction_MathTransformMakeRelative(
    Global=(Rotation=(X=0,Y=0,Z=0,W=1),Translation=(X=0,Y=0,Z=0),Scale3D=(X=1,Y=1,Z=1))
    Parent=@n2.Transform
    Global.Translation=&RigVMFunction_MathTransformMul(
     A=(Rotation=(X=0,Y=0,Z=0,W=1),Translation=(X=0,Y=0,Z=0),Scale3D=(X=1,Y=1,Z=1))
     B=@n3.Transform
     A.Translation.Y=&RigVMFunction_MathFloatClamp(Maximum=0, Minimum=-100, Value=@n6.Local.Translation.Y) {template="Clamp::Execute(in Value,in Minimum,in Maximum,out Result)"}.Result
     A.Translation.X=@n6.Local.Translation.X
     A.Translation.Z=@n6.Local.Translation.Z
    ) {template="Multiply::Execute(in A,in B,out Result)"}.Result.Translation
   )
   /Script/ControlRig.RigUnit_SetBoneTranslation(
    Bone=hairTipMid_side_l
    Space=GlobalSpace
    Weight=1
    bPropagateToChildren=False
    ExecutePin=@n1.ExecutePin
    Translation=&RigVMFunction_MathTransformMul(
     A=(Rotation=(X=0,Y=0,Z=0,W=1),Translation=(X=0,Y=0,Z=0),Scale3D=(X=1,Y=1,Z=1))
     B=@n2.Transform
     A.Translation.Y=&RigVMFunction_MathFloatClamp(Maximum=0, Minimum=-100, Value=@n7.Local.Translation.Y) {template="Clamp::Execute(in Value,in Minimum,in Maximum,out Result)"}.Result
     A.Translation.X=@n7.Local.Translation.X
     A.Translation.Z=@n7.Local.Translation.Z
    ) {template="Multiply::Execute(in A,in B,out Result)"}.Result.Translation
   ) {is=struct RigVMUnitNode}
   n8 = &RigUnit_GetBoneTransform(Bone=chestCol_c)
   n9 = &RigUnit_GetBoneTransform(Bone=chestCol_r)
   n10 = &RigUnit_GetBoneTransform(Bone=chestCol_l)
   n11 = &RigVMFunction_MathTransformMakeRelative(Global=&RigUnit_GetBoneTransform(Bone=hairTipMid_side_r).Transform, Parent=@n10.Transform)
   n12 = &RigVMFunction_MathTransformMakeRelative(
    Global=(Rotation=(X=0,Y=0,Z=0,W=1),Translation=(X=0,Y=0,Z=0),Scale3D=(X=1,Y=1,Z=1))
    Parent=@n9.Transform
    Global.Translation=&RigVMFunction_MathTransformMul(
     A=(Rotation=(X=0,Y=0,Z=0,W=1),Translation=(X=0,Y=0,Z=0),Scale3D=(X=1,Y=1,Z=1))
     B=@n10.Transform
     A.Translation.Y=&RigVMFunction_MathFloatClamp(Maximum=0, Minimum=-100, Value=@n11.Local.Translation.Y) {template="Clamp::Execute(in Value,in Minimum,in Maximum,out Result)"}.Result
     A.Translation.X=@n11.Local.Translation.X
     A.Translation.Z=@n11.Local.Translation.Z
    ) {template="Multiply::Execute(in A,in B,out Result)"}.Result.Translation
   )
   n13 = &RigVMFunction_MathTransformMakeRelative(
    Global=(Rotation=(X=0,Y=0,Z=0,W=1),Translation=(X=0,Y=0,Z=0),Scale3D=(X=1,Y=1,Z=1))
    Parent=@n8.Transform
    Global.Translation=&RigVMFunction_MathTransformMul(
     A=(Rotation=(X=0,Y=0,Z=0,W=1),Translation=(X=0,Y=0,Z=0),Scale3D=(X=1,Y=1,Z=1))
     B=@n9.Transform
     A.Translation.Y=&RigVMFunction_MathFloatClamp(Maximum=0, Minimum=-100, Value=@n12.Local.Translation.Y) {template="Clamp::Execute(in Value,in Minimum,in Maximum,out Result)"}.Result
     A.Translation.X=@n12.Local.Translation.X
     A.Translation.Z=@n12.Local.Translation.Z
    ) {template="Multiply::Execute(in A,in B,out Result)"}.Result.Translation
   )
   > &RigUnit_SetBoneTranslation(
    Bone=hairTipMid_side_r
    Space=GlobalSpace
    Weight=1
    bPropagateToChildren=False
    Translation=&RigVMFunction_MathTransformMul(
     A=(Rotation=(X=0,Y=0,Z=0,W=1),Translation=(X=0,Y=0,Z=0),Scale3D=(X=1,Y=1,Z=1))
     B=@n8.Transform
     A.Translation.Y=&RigVMFunction_MathFloatClamp(Maximum=0, Minimum=-100, Value=@n13.Local.Translation.Y) {template="Clamp::Execute(in Value,in Minimum,in Maximum,out Result)"}.Result
     A.Translation.X=@n13.Local.Translation.X
     A.Translation.Z=@n13.Local.Translation.Z
    ) {template="Multiply::Execute(in A,in B,out Result)"}.Result.Translation
   )
   n14 = &RigUnit_GetBoneTransform(Bone=chestCol_c)
   n15 = &RigUnit_GetBoneTransform(Bone=chestCol_2_r)
   n16 = &RigVMFunction_MathTransformMakeRelative(Global=&RigUnit_GetBoneTransform(Bone=hair_3_side_r).Transform, Parent=@n15.Transform)
   n17 = &RigVMFunction_MathTransformMakeRelative(
    Global=(Rotation=(X=0,Y=0,Z=0,W=1),Translation=(X=0,Y=0,Z=0),Scale3D=(X=1,Y=1,Z=1))
    Parent=@n14.Transform
    Global.Translation=&RigVMFunction_MathTransformMul(
     A=(Rotation=(X=0,Y=0,Z=0,W=1),Translation=(X=0,Y=0,Z=0),Scale3D=(X=1,Y=1,Z=1))
     B=@n15.Transform
     A.Translation.Y=&RigVMFunction_MathFloatClamp(Maximum=0, Minimum=-100, Value=@n16.Local.Translation.Y) {template="Clamp::Execute(in Value,in Minimum,in Maximum,out Result)"}.Result
     A.Translation.X=@n16.Local.Translation.X
     A.Translation.Z=@n16.Local.Translation.Z
    ) {template="Multiply::Execute(in A,in B,out Result)"}.Result.Translation
   )
   > &RigUnit_SetBoneTranslation(
    Bone=hair_3_side_r
    Space=GlobalSpace
    Weight=1
    bPropagateToChildren=False
    Translation=&RigVMFunction_MathTransformMul(
     A=(Rotation=(X=0,Y=0,Z=0,W=1),Translation=(X=0,Y=0,Z=0),Scale3D=(X=1,Y=1,Z=1))
     B=@n14.Transform
     A.Translation.Y=&RigVMFunction_MathFloatClamp(Maximum=0, Minimum=-100, Value=@n17.Local.Translation.Y) {template="Clamp::Execute(in Value,in Minimum,in Maximum,out Result)"}.Result
     A.Translation.X=@n17.Local.Translation.X
     A.Translation.Z=@n17.Local.Translation.Z
    ) {template="Multiply::Execute(in A,in B,out Result)"}.Result.Translation
   )
   n18 = &RigUnit_GetBoneTransform(Bone=chestCol_c)
   n19 = &RigUnit_GetBoneTransform(Bone=chestCol_2_l)
   n20 = &RigVMFunction_MathTransformMakeRelative(Global=&RigUnit_GetBoneTransform(Bone=hair_3_side_l).Transform, Parent=@n19.Transform)
   n21 = &RigVMFunction_MathTransformMakeRelative(
    Global=(Rotation=(X=0,Y=0,Z=0,W=1),Translation=(X=0,Y=0,Z=0),Scale3D=(X=1,Y=1,Z=1))
    Parent=@n18.Transform
    Global.Translation=&RigVMFunction_MathTransformMul(
     A=(Rotation=(X=0,Y=0,Z=0,W=1),Translation=(X=0,Y=0,Z=0),Scale3D=(X=1,Y=1,Z=1))
     B=@n19.Transform
     A.Translation.Y=&RigVMFunction_MathFloatClamp(Maximum=0, Minimum=-100, Value=@n20.Local.Translation.Y) {template="Clamp::Execute(in Value,in Minimum,in Maximum,out Result)"}.Result
     A.Translation.X=@n20.Local.Translation.X
     A.Translation.Z=@n20.Local.Translation.Z
    ) {template="Multiply::Execute(in A,in B,out Result)"}.Result.Translation
   )
   > &RigUnit_SetBoneTranslation(
    Bone=hair_3_side_l
    Space=GlobalSpace
    Weight=1
    bPropagateToChildren=True
    Translation=&RigVMFunction_MathTransformMul(
     A=(Rotation=(X=0,Y=0,Z=0,W=1),Translation=(X=0,Y=0,Z=0),Scale3D=(X=1,Y=1,Z=1))
     B=@n18.Transform
     A.Translation.Y=&RigVMFunction_MathFloatClamp(Maximum=0, Minimum=-100, Value=@n21.Local.Translation.Y) {template="Clamp::Execute(in Value,in Minimum,in Maximum,out Result)"}.Result
     A.Translation.X=@n21.Local.Translation.X
     A.Translation.Z=@n21.Local.Translation.Z
    ) {template="Multiply::Execute(in A,in B,out Result)"}.Result.Translation
   )
   n22 = &RigUnit_GetBoneTransform(Bone=chestCol_c)
   n23 = &RigUnit_GetBoneTransform(Bone=chestCol_r)
   n24 = &RigUnit_GetBoneTransform(Bone=chestCol_l)
   n25 = &RigVMFunction_MathTransformMakeRelative(Global=&RigUnit_GetBoneTransform(Bone=hairTipMid_l_c).Transform, Parent=@n24.Transform)
   n26 = &RigVMFunction_MathTransformMakeRelative(
    Global=(Rotation=(X=0,Y=0,Z=0,W=1),Translation=(X=0,Y=0,Z=0),Scale3D=(X=1,Y=1,Z=1))
    Parent=@n23.Transform
    Global.Translation=&RigVMFunction_MathTransformMul(
     A=(Rotation=(X=0,Y=0,Z=0,W=1),Translation=(X=0,Y=0,Z=0),Scale3D=(X=1,Y=1,Z=1))
     B=@n24.Transform
     A.Translation.Y=&RigVMFunction_MathFloatClamp(Maximum=0, Minimum=-100, Value=@n25.Local.Translation.Y) {template="Clamp::Execute(in Value,in Minimum,in Maximum,out Result)"}.Result
     A.Translation.X=@n25.Local.Translation.X
     A.Translation.Z=@n25.Local.Translation.Z
    ) {template="Multiply::Execute(in A,in B,out Result)"}.Result.Translation
   )
   n27 = &RigVMFunction_MathTransformMakeRelative(
    Global=(Rotation=(X=0,Y=0,Z=0,W=1),Translation=(X=0,Y=0,Z=0),Scale3D=(X=1,Y=1,Z=1))
    Parent=@n22.Transform
    Global.Translation=&RigVMFunction_MathTransformMul(
     A=(Rotation=(X=0,Y=0,Z=0,W=1),Translation=(X=0,Y=0,Z=0),Scale3D=(X=1,Y=1,Z=1))
     B=@n23.Transform
     A.Translation.Y=&RigVMFunction_MathFloatClamp(Maximum=0, Minimum=-100, Value=@n26.Local.Translation.Y) {template="Clamp::Execute(in Value,in Minimum,in Maximum,out Result)"}.Result
     A.Translation.X=@n26.Local.Translation.X
     A.Translation.Z=@n26.Local.Translation.Z
    ) {template="Multiply::Execute(in A,in B,out Result)"}.Result.Translation
   )
   > &RigUnit_SetBoneTranslation(
    Bone=hairTipMid_l_c
    Space=GlobalSpace
    Weight=1
    bPropagateToChildren=False
    Translation=&RigVMFunction_MathTransformMul(
     A=(Rotation=(X=0,Y=0,Z=0,W=1),Translation=(X=0,Y=0,Z=0),Scale3D=(X=1,Y=1,Z=1))
     B=@n22.Transform
     A.Translation.Y=&RigVMFunction_MathFloatClamp(Maximum=0, Minimum=-100, Value=@n27.Local.Translation.Y) {template="Clamp::Execute(in Value,in Minimum,in Maximum,out Result)"}.Result
     A.Translation.X=@n27.Local.Translation.X
     A.Translation.Z=@n27.Local.Translation.Z
    ) {template="Multiply::Execute(in A,in B,out Result)"}.Result.Translation
   )
   n28 = &RigUnit_GetBoneTransform(Bone=chestCol_c)
   n29 = &RigUnit_GetBoneTransform(Bone=chestCol_2_r)
   n30 = &RigUnit_GetBoneTransform(Bone=chestCol_l)
   n31 = &RigVMFunction_MathTransformMakeRelative(Global=&RigUnit_GetBoneTransform(Bone=hairTipMid_r_c).Transform, Parent=@n30.Transform)
   n32 = &RigVMFunction_MathTransformMakeRelative(
    Global=(Rotation=(X=0,Y=0,Z=0,W=1),Translation=(X=0,Y=0,Z=0),Scale3D=(X=1,Y=1,Z=1))
    Parent=@n29.Transform
    Global.Translation=&RigVMFunction_MathTransformMul(
     A=(Rotation=(X=0,Y=0,Z=0,W=1),Translation=(X=0,Y=0,Z=0),Scale3D=(X=1,Y=1,Z=1))
     B=@n30.Transform
     A.Translation.Y=&RigVMFunction_MathFloatClamp(Maximum=0, Minimum=-100, Value=@n31.Local.Translation.Y) {template="Clamp::Execute(in Value,in Minimum,in Maximum,out Result)"}.Result
     A.Translation.X=@n31.Local.Translation.X
     A.Translation.Z=@n31.Local.Translation.Z
    ) {template="Multiply::Execute(in A,in B,out Result)"}.Result.Translation
   )
   n33 = &RigVMFunction_MathTransformMakeRelative(
    Global=(Rotation=(X=0,Y=0,Z=0,W=1),Translation=(X=0,Y=0,Z=0),Scale3D=(X=1,Y=1,Z=1))
    Parent=@n28.Transform
    Global.Translation=&RigVMFunction_MathTransformMul(
     A=(Rotation=(X=0,Y=0,Z=0,W=1),Translation=(X=0,Y=0,Z=0),Scale3D=(X=1,Y=1,Z=1))
     B=@n29.Transform
     A.Translation.Y=&RigVMFunction_MathFloatClamp(Maximum=0, Minimum=-100, Value=@n32.Local.Translation.Y) {template="Clamp::Execute(in Value,in Minimum,in Maximum,out Result)"}.Result
     A.Translation.X=@n32.Local.Translation.X
     A.Translation.Z=@n32.Local.Translation.Z
    ) {template="Multiply::Execute(in A,in B,out Result)"}.Result.Translation
   )
   > &RigUnit_SetBoneTranslation(
    Bone=hairTipMid_r_c
    Space=GlobalSpace
    Weight=1
    bPropagateToChildren=False
    Translation=&RigVMFunction_MathTransformMul(
     A=(Rotation=(X=0,Y=0,Z=0,W=1),Translation=(X=0,Y=0,Z=0),Scale3D=(X=1,Y=1,Z=1))
     B=@n28.Transform
     A.Translation.Y=&RigVMFunction_MathFloatClamp(Maximum=0, Minimum=-100, Value=@n33.Local.Translation.Y) {template="Clamp::Execute(in Value,in Minimum,in Maximum,out Result)"}.Result
     A.Translation.X=@n33.Local.Translation.X
     A.Translation.Z=@n33.Local.Translation.Z
    ) {template="Multiply::Execute(in A,in B,out Result)"}.Result.Translation
   )
   n34 = &RigUnit_GetBoneTransform(Bone=chestCol_c)
   n35 = &RigUnit_GetBoneTransform(Bone=chestCol_2_l)
   n36 = &RigVMFunction_MathTransformMakeRelative(Global=&RigUnit_GetBoneTransform(Bone=hair_3_side2_l).Transform, Parent=@n35.Transform)
   n37 = &RigVMFunction_MathTransformMakeRelative(
    Global=(Rotation=(X=0,Y=0,Z=0,W=1),Translation=(X=0,Y=0,Z=0),Scale3D=(X=1,Y=1,Z=1))
    Parent=@n34.Transform
    Global.Translation=&RigVMFunction_MathTransformMul(
     A=(Rotation=(X=0,Y=0,Z=0,W=1),Translation=(X=0,Y=0,Z=0),Scale3D=(X=1,Y=1,Z=1))
     B=@n35.Transform
     A.Translation.Y=&RigVMFunction_MathFloatClamp(Maximum=0, Minimum=-100, Value=@n36.Local.Translation.Y) {template="Clamp::Execute(in Value,in Minimum,in Maximum,out Result)"}.Result
     A.Translation.X=@n36.Local.Translation.X
     A.Translation.Z=@n36.Local.Translation.Z
    ) {template="Multiply::Execute(in A,in B,out Result)"}.Result.Translation
   )
   > &RigUnit_SetBoneTranslation(
    Bone=hair_3_side2_l
    Space=GlobalSpace
    Weight=1
    bPropagateToChildren=True
    Translation=&RigVMFunction_MathTransformMul(
     A=(Rotation=(X=0,Y=0,Z=0,W=1),Translation=(X=0,Y=0,Z=0),Scale3D=(X=1,Y=1,Z=1))
     B=@n34.Transform
     A.Translation.Y=&RigVMFunction_MathFloatClamp(Maximum=0, Minimum=-100, Value=@n37.Local.Translation.Y) {template="Clamp::Execute(in Value,in Minimum,in Maximum,out Result)"}.Result
     A.Translation.X=@n37.Local.Translation.X
     A.Translation.Z=@n37.Local.Translation.Z
    ) {template="Multiply::Execute(in A,in B,out Result)"}.Result.Translation
   )
   > /Script/ControlRig.RigUnit_AimBone(
    Bone=hair_2_l
    DebugSettings=(bEnabled=False,Scale=10,WorldOffset=(Rotation=(X=0,Y=0,Z=0,W=1),Translation=(X=0,Y=0,Z=0),Scale3D=(X=1,Y=1,Z=1)))
    Primary=(Weight=1,Axis=(X=12,Y=2,Z=5),Target=(X=0,Y=0,Z=0),Kind=Location,Space="hair_3_side2_l")
    Secondary=(Weight=0,Axis=(X=0,Y=0,Z=1),Target=(X=1,Y=0,Z=0),Kind=Direction,Space="None")
    Weight=1
    bPropagateToChildren=False
   ) {is=struct RigVMUnitNode}
   n38 = &RigUnit_GetBoneTransform(Bone=hair_2_side_r)
   n39 = &RigUnit_GetBoneTransform(Bone=chestCol_2_r)
   n40 = &RigVMFunction_MathTransformMakeRelative(Global=@n38.Transform, Parent=@n39.Transform)
   n41 = &RigVMFunction_MathFloatClamp(Maximum=0, Minimum=-100, Value=@n40.Local.Translation.Y) {template="Clamp::Execute(in Value,in Minimum,in Maximum,out Result)"}
   > &RigUnit_SetBoneTranslation(
    Bone=hair_2_side_r
    Space=GlobalSpace
    Weight=1
    bPropagateToChildren=False
    Translation=/Script/RigVM.RigVMFunction_MathVectorLerp(
     T=0.4
     A=@n38.Transform.Translation
     B=&RigVMFunction_MathTransformMul(
      A=(Rotation=(X=0,Y=0,Z=0,W=1),Translation=(X=0,Y=0,Z=0),Scale3D=(X=1,Y=1,Z=1))
      B=@n39.Transform
      A.Translation.Y=@n41.Result
      A.Translation.X=@n40.Local.Translation.X
      A.Translation.Z=@n40.Local.Translation.Z
     ) {template="Multiply::Execute(in A,in B,out Result)"}.Result.Translation
    ) {is=struct RigVMUnitNode, template="Interpolate::Execute(in A,in B,in T,out Result)"}.Result
   )
  ]}
]
hierarchy [
 root {loc=<0 0 0>, rot=<0 0 0 1>}
 pelvis {parent=root, loc=<0 -0.5301 98.0571>, rot=<0.0127 0.707 -0.0127 -0.707>}
 spine_01 {parent=pelvis, loc=<6.1208 -0.9328 0>, rot=<0 0 0.034 -0.9994>}
 spine_02 {parent=spine_01, loc=<6.5996 -0.2095 0>, rot=<0 0 0.0238 -0.9997>}
 spine_03 {parent=spine_02, loc=<6.4914 -0.1237 0>, rot=<0 0 0.0149 -0.9999>}
 spine_04 {parent=spine_03, loc=<6.9598 0.0794 0>, rot=<0 0 -0.0132 -0.9999>}
 spine_05 {parent=spine_04, loc=<6.9987 0 0>, rot=<0 0 -0.0015 -1>}
 clavicle_l {parent=spine_05, loc=<5.4831 -0.1053 -3.2946>, rot=<0.0465 -0.7289 -0.009 -0.6829>}
 upperarm_l {parent=clavicle_l, loc=<9.8658 -1.2136 -0.4105>, rot=<0.0164 -0.3245 -0.0269 -0.9454>}
 upperarm_twist_01_l {parent=upperarm_l, loc=<7.2135 -0.1134 0>, rot=<0 0 0 1>}
 clavicle_r {parent=spine_05, loc=<5.4831 -0.1053 3.2946>, rot=<0.7289 0.0465 -0.6829 0.009>}
 upperarm_r {parent=clavicle_r, loc=<-9.8658 1.2136 0.4104>, rot=<0.0164 -0.3245 -0.0269 -0.9454>}
 upperarm_twist_01_r {parent=upperarm_r, loc=<-7.2135 0.1134 0>, rot=<0 0 0 1>}
 neck_01 {parent=spine_05, loc=<11.3138 0.1529 0>, rot=<0 0 -0.2003 -0.9797>}
 neck_02 {parent=neck_01, loc=<4.5627 0 0>, rot=<0 0 0.0504 -0.9987>}
 head {parent=neck_02, loc=<5.3914 0 0>, rot=<0 0 0.115 -0.9934>}
 hair_head {parent=head, loc=<8.3535 -1.209 0>, rot=<0.097 -0.9953 0 0>}
 hairPivot_aim_r {parent=hair_head, loc=<0 0 -6.5054>, rot=<0 -0.0598 0.0912 -0.994>}
 hair_1_r {parent=hairPivot_aim_r, loc=<0 0 0>, rot=<0 0.0625 -0.0289 -0.9976>}
 hair_2_r {parent=hair_1_r, loc=<12.8252 0 0>, rot=<0.0059 -0.0073 0.0978 -0.9952>}
 hair_3_r {parent=hair_2_r, loc=<7.0201 0 0>, rot=<-0.0544 -0.0714 -0.0039 -0.996>}
 hair_3_side_r {parent=hair_3_r, loc=<3.8045 0.4235 -2.3138>, rot=<0 0 0 1>}
 hairTipMid_r {parent=hair_2_r, loc=<7.6219 -1.1181 1.5615>, rot=<-0.0092 0.0047 -0.1596 -0.9871>}
 hairTipMid_r_c {parent=hairTipMid_r, loc=<7.3773 0.3929 5.3185>, rot=<0 0 0 -1>}
 hairTipMid_side_r {parent=hairTipMid_r, loc=<5.3034 0.5608 -0.8969>, rot=<0 0 0 -1>}
 hair_2_side_r {parent=hair_2_r, loc=<3.8555 1.9477 -2.2509>, rot=<0 0 0 1>}
 hair_1_side_r {parent=hair_1_r, loc=<11.0844 1.5973 -2.0387>, rot=<-0.1123 0.6971 -0.1064 -0.7001>}
 hairPivot_aim_l {parent=hair_head, loc=<0 0 6.5054>, rot=<0 0.0967 0.0778 -0.9923>}
 hair_1_l {parent=hairPivot_aim_l, loc=<0 0 0>, rot=<0 -0.0994 -0.0153 -0.9949>}
 hair_2_l {parent=hair_1_l, loc=<12.8252 0 0>, rot=<-0.0101 -0.0108 0.0998 -0.9949>}
 hair_3_l {parent=hair_2_l, loc=<9.2191 0 0>, rot=<0.0334 0.0928 -0.164 -0.9815>}
 hair_3_side_l {parent=hair_3_l, loc=<4.1461 0.2478 0.1332>, rot=<0 0 0 -1>}
 hair_3_side2_l {parent=hair_3_l, loc=<2.9177 0.3263 3.9044>, rot=<0 0 0 -1>}
 hairTipMid_l {parent=hair_2_l, loc=<7.6798 -1.0983 -1.2635>, rot=<0.0168 0.0131 -0.1615 -0.9866>}
 hairTipMid_l_c {parent=hairTipMid_l, loc=<7.3773 0.3929 -4.4826>, rot=<0 0 0 -1>}
 hairTipMid_side_l {parent=hairTipMid_l, loc=<6.5333 0.7391 1>, rot=<0 0 0 -1>}
 null1 {parent=hair_2_l, loc=<12.9666 1.5994 0.8169>, rot=<0.0334 0.0928 -0.164 -0.9815>}
 hairPivot_tgt_l {parent=spine_05, loc=<7.1356 -6.9674 -11.2136>, rot=<0.2112 -0.9768 -0.0073 -0.0336>}
 hairPivot_tgt_r {parent=spine_05, loc=<8.1525 -7.1585 9.2854>, rot=<0.2112 -0.9768 0.0073 0.0336>}
 chestCol_r {parent=spine_05, loc=<8.2428 -5.8624 7.547>, rot=<0.1203 -0.6968 -0.1203 -0.6968>}
 chestCol_c {parent=spine_05, loc=<7.4789 -6.8059 -0.0109>, rot=<0.1203 -0.6968 -0.1203 -0.6968>}
 chestCol_l {parent=spine_05, loc=<8.3121 -5.8576 -7.2697>, rot=<0.1203 -0.6968 -0.1203 -0.6968>}
 chestCol_2_r {parent=spine_05, loc=<6.6342 -6.5031 12.4542>, rot=<0.1203 -0.6968 -0.1203 -0.6968>}
 chestCol_2_l {parent=spine_05, loc=<6.5299 -6.5657 -12.4907>, rot=<0.1203 -0.6968 -0.1203 -0.6968>}
]
